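; Grenade target prop for Outpost missions: a troll cutout that animates into place, then is destroyed shortly after a Commando frag grenade overlaps it.
; On BeginPlay a timeline drives the cutout's height and scale through UpdateLocation, firing trail, sound, impact and post-process effects, then pings the minimap.
; A capsule overlap starts a timer that destroys the actor and sets GrenadeHit, which gates ShouldDie and triggers SilentDie on replication.
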(event EventBeginPlay
  (bind _trail (Variables|Default|GetTrail))
  (Utilities|FlowControl|DoOnce false
    (:Completed
      (Audio|PlaySoundatLocation "/Game/Sounds/Fort_GamePlay_Sounds/Missions/slammer_start_Cue.slammer_start_Cue" (Transformation|GetActorLocation) "0, 0, 0" 1.2 1.6 0.0 "/Game/Sounds/Attenuation/Fort_TileAmbient_NoAtt.Fort_TileAmbient_NoAtt")
      (bind (_direction _newtrack_1 _location) (Timeline|Timeline_1 PlayFromStart)
        (:Update
          (Utilities|FlowControl|Sequence
            (:then_0
              ;[1 0.4913 0] Start the Drop
              (Utilities|FlowControl|DoOnce false
                (:Completed
                  (Rendering|SetVisibility (Variables|Default|GetSM_Troll_Cutout) true false)
                  (Components|Activation|Activate (Variables|Default|GetNS_SupplyDrop_Active) false))))
            (:then_1
              (CallFunction|UpdateLocation _newtrack_1 _location))))
        (:Finished
          ;[0.3534 0.05274 1] Trail Effects
          (Components|Activation|Deactivate _trail)
          ;[1 0 0.009008] Comment
          (MissionHelpers|PingMinimap (Variables|Default|GetImpactLocation) "(R=1.000000,G=0.673128,B=0.002336,A=1.000000)")
          (Rendering|SetVisibility (Variables|Debug|GetStaticMeshComponent) true false)
          (Rendering|SetVisibility (Variables|Default|GetSM_Troll_Cutout) false false)
          (Development|SetHiddeninGame (Variables|Transformation|GetRootComponent) false false)
          (Components|Activation|Deactivate (Variables|Default|GetNS_SupplyDrop_Active))
          (Components|Activation|Activate (Variables|Default|GetP_InteractionBang) false))
        (:Impact
          ;[0.3465 0.07805 1] Impact
          (Utilities|FlowControl|DoOnce false
            (:Completed
              (Effects|SpawnEmitteratLocation "/Game/Gadgets/Assets/SupplyDrop/Effects/P_SupplyDrop_Impact.P_SupplyDrop_Impact" (Variables|Default|GetImpactLocation)) ; Impact Effect
              ;[0.09479 0.3076 1] Post and Shake
              (Networking|CanExecuteCosmeticEvents
                (:True
                  (bind _returnvalue (Game|SpawnActorfromClass "/Game/Effects/Fort_Effects/BluePrints/PostProcessFX/B_Post_SupplyDrop_Impact.B_Post_SupplyDrop_Impact_C" (Transformation|GetActorTransform) "AlwaysSpawn" "OverrideRootScale" 0 "400.000000,400.000000,400.000000" 0))
                  (Class|BPostSupplyDropImpact|SpawnPostFX _returnvalue 1.3e+03 (Variables|Default|GetImpactLocation) true 2 false))))))
        (:StartTrail
          (Components|Activation|Activate _trail false)
          (Utilities|FlowControl|Delay 2)
          (Components|Activation|Activate (Variables|Default|GetAtlasSent) false)
          (Audio|PlaySoundatLocation "/Game/Sounds/Fort_Gadgets/Teleporter/teleport_Cue.teleport_Cue" (Transformation|GetActorLocation) "0, 0, 0" 1.0 1.0 0.0 "/Game/Sounds/Attenuation/Fort_TileAmbient_NoAtt.Fort_TileAmbient_NoAtt"))))))

(event Collision|EventActorBeginOverlap
  (params (OtherActor :type Actor%20Object%20Reference)))

(event EventTick
  (params (DeltaSeconds :type Float%20%28single-precision%29)))

(event OnComponentBeginOverlap(Capsule)
  (params (OverlappedComponent :type Primitive%20Component%20Object%20Reference) (OtherActor :type Actor%20Object%20Reference) (OtherComp :type Primitive%20Component%20Object%20Reference) (OtherBodyIndex :type Integer) (bFromSweep :type Boolean) (SweepResult :type Hit%20Result%20Structure%20%28by%20ref%29))
  (if (Utilities|Equal(Class) (Utilities|GetClass OtherActor) "/Game/Abilities/Player/Commando/Actives/FragGrenade/B_Prj_Commando_FragGrenade.B_Prj_Commando_FragGrenade_C")
    (Utilities|Time|SetTimerbyEvent (delegate Custom|DestroySelf) 0.75)))

(event Custom|DestroySelf
  (Actor|DestroyActor)
  (Variables|Default|SetGrenadeHit true))

(fn OnRep_GrenadeHit
  (Gameplay|SilentDie false))

(fn ShouldDie
  (params (Damage :type Float%20%28single-precision%29) (EventInstigator :type Controller%20Object%20Reference) (DamageCauser :type Actor%20Object%20Reference) (bIsTest :type Boolean) (EffectContext :type Gameplay%20Effect%20Context%20Handle%20Structure%20%28by%20ref%29))
  (return (Variables|Default|GetGrenadeHit)))

(fn UpdateLocation
  (params (Time :type Float%20%28double-precision%29) (Location :type Float%20%28double-precision%29))
  (bind _returnvalue (Transformation|GetActorTransform))
  (bind _location (.location _returnvalue))
  (bind _sm_troll_cutout (Variables|Default|GetSM_Troll_Cutout))
  (Variables|Default|SetImpactLocation _location)
  (Transformation|SetWorldScale3D _sm_troll_cutout (Math|Vector|Lerp(Vector) "1.000000,1.000000,1.000000" "0.500000,0.500000,0.500000" Location))
  (Transformation|SetWorldLocation _sm_troll_cutout (Math|Vector|Lerp(Vector) (Math|Vector|MakeVector (.x _location) (.y _location) (- (.z _location) 1.0)) (Math|Vector|MakeVector (.x _location) (.y _location) (+ (.z _location) 128)) Time) false false))

(fn ConstructionScript)
